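# Looping ambient sound cue for the high-voltage hum of an exterior power plant pylon prop, judging by its path and wave name.
# A single looping wave player feeds the root directly, at reduced volume on the Fort_Ambient_TileMix sound class.
# Overridden attenuation turns off spatialization, so the sound fades with distance from its source without being panned in 3D.
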SoundCue /Game/Sounds/Fort_Prop_Sounds/Exterior/PowerPlant_Pylon/Tower_HighVoltage_Cue
root=WavePlayer_0
nodes [{id=WavePlayer_0, properties {bLooping=true, soundWaveAssetPtr=&/Tower_HighVoltage_01}}]
properties {
 bOverrideAttenuation=true
 maxDistance=3450.9424
 soundClassObject=/Game/Sounds/FortSoundClassesAndModes/Fort_Ambient_TileMix
 volumeMultiplier=0.5
 attenuationOverrides {bSpatialize=false, falloffDistance=1961.4901, stereoSpread=1000, attenuationShapeExtents {x=1489.4523}}
}
